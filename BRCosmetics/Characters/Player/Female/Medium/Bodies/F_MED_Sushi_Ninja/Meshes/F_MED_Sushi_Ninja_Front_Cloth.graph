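# Physics asset for the front apron cloth of the female Sushi Ninja character, simulating it with rigid-body bone chains.
# Left and right three-box chains hang from small driver bodies that are angularly locked to the pelvis, with angular limits generally loosening below the first joint.
# Cross constraints tie the two sides together at the upper two segments, leaving the bottom pair unlinked.
# Pelvis and thigh capsules, also angularly locked together, supply body collision shapes.
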PhysicsAsset /BRCosmetics/Characters/Player/Female/Medium/Bodies/F_MED_Sushi_Ninja/Meshes/F_MED_Sushi_Ninja_Front_Cloth
bodies {
 driver_apron_front_01_l [{shapeType=Box, extentX=1.01, extentY=1.01, extentZ=1.01, shapeName=driver_apron_front_01_l_box, rotation=<rot 0 -90 0>}]
 driver_apron_front_01_r [{shapeType=Box, extentX=1.01, extentY=1.01, extentZ=1.01, shapeName=driver_apron_front_01_r_box, rotation=<rot 0 -90 0>}]
 dyn_apron_frnt_rb_01_l [{shapeType=Box, extentX=10, extentY=1.5, extentZ=6, shapeName=dyn_apron_frnt_rb_01_l_box, center=<2 0.5595 -0.7>, rotation=<rot 90 0 10.0002>}]
 dyn_apron_frnt_rb_01_r [{
   shapeType=Box
   extentX=10
   extentY=1.5
   extentZ=6
   shapeName=dyn_apron_frnt_rb_01_r_box
   center=<2 0.6307 0.7>
   rotation=<rot 90 -179.1048 -174.1047>
  }]
 dyn_apron_frnt_rb_02_l [{shapeType=Box, extentX=10, extentY=1, extentZ=11, shapeName=dyn_apron_frnt_rb_02_l_box, center=<2 0.2228 -1.3>}]
 dyn_apron_frnt_rb_02_r [{shapeType=Box, extentX=10, extentY=1, extentZ=11, shapeName=dyn_apron_frnt_rb_02_r_box, center=<2 0.2228 1.3>}]
 dyn_apron_frnt_rb_03_l [{shapeType=Box, extentX=13, extentY=2, extentZ=13, shapeName=dyn_apron_frnt_rb_03_l_box, center=<4.3686 0.3992 -2>, rotation=<rot 90 0 0>}]
 dyn_apron_frnt_rb_03_r [{shapeType=Box, extentX=13, extentY=2, extentZ=13, shapeName=dyn_apron_frnt_rb_03_r_box, center=<4.3686 0.3992 2>, rotation=<rot 90 0 0>}]
 pelvis [
  {shapeType=Capsule, length=15, radius=9, shapeName=pelvis_capsule, center=<-2 -0.4933 0>}
  {shapeType=Capsule, length=15, radius=9, shapeName=pelvis_capsule, center=<-7.9179 -0.4359 0>}
 ]
 thigh_l [{shapeType=Capsule, length=25.5, radius=9, shapeName=thigh_l_capsule, center=<-16 0.3563 0>, rotation=<rot 90 0 0>}]
 thigh_r [{shapeType=Capsule, length=25.5, radius=9, shapeName=thigh_r_capsule, center=<17 -0.2 0>, rotation=<rot 90 0 0>}]
}
constraints [
 {
  bone1Name=thigh_l
  bone2Name=pelvis
  swing1LimitDegrees=45
  swing1Motion=Locked
  swing2LimitDegrees=45
  swing2Motion=Locked
  twistLimitDegrees=45
  twistMotion=Locked
 }
 {
  bone1Name=thigh_r
  bone2Name=pelvis
  swing1LimitDegrees=45
  swing1Motion=Locked
  swing2LimitDegrees=45
  swing2Motion=Locked
  twistLimitDegrees=45
  twistMotion=Locked
 }
 {
  bone1Name=driver_apron_front_01_l
  bone2Name=pelvis
  swing1LimitDegrees=45
  swing1Motion=Locked
  swing2LimitDegrees=45
  swing2Motion=Locked
  twistLimitDegrees=45
  twistMotion=Locked
 }
 {bone1Name=dyn_apron_frnt_rb_01_l, bone2Name=driver_apron_front_01_l, swing1LimitDegrees=12, swing2LimitDegrees=15, twistLimitDegrees=15}
 {bone1Name=dyn_apron_frnt_rb_02_l, bone2Name=dyn_apron_frnt_rb_01_l, swing1LimitDegrees=20, swing2LimitDegrees=20, twistLimitDegrees=45}
 {bone1Name=dyn_apron_frnt_rb_03_l, bone2Name=dyn_apron_frnt_rb_02_l, swing1LimitDegrees=30, swing2LimitDegrees=20, twistLimitDegrees=20}
 {
  bone1Name=driver_apron_front_01_r
  bone2Name=pelvis
  swing1LimitDegrees=45
  swing1Motion=Locked
  swing2LimitDegrees=45
  swing2Motion=Locked
  twistLimitDegrees=45
  twistMotion=Locked
 }
 {bone1Name=dyn_apron_frnt_rb_01_r, bone2Name=driver_apron_front_01_r, swing1LimitDegrees=12, swing2LimitDegrees=15, twistLimitDegrees=15}
 {bone1Name=dyn_apron_frnt_rb_02_r, bone2Name=dyn_apron_frnt_rb_01_r, swing1LimitDegrees=20, swing2LimitDegrees=20, twistLimitDegrees=20}
 {bone1Name=dyn_apron_frnt_rb_03_r, bone2Name=dyn_apron_frnt_rb_02_r, swing1LimitDegrees=30, swing2LimitDegrees=20, twistLimitDegrees=20}
 {bone1Name=dyn_apron_frnt_rb_01_r, bone2Name=dyn_apron_frnt_rb_01_l, swing1LimitDegrees=15, swing2LimitDegrees=15, twistLimitDegrees=15}
 {bone1Name=dyn_apron_frnt_rb_02_r, bone2Name=dyn_apron_frnt_rb_02_l, swing1LimitDegrees=15, swing2LimitDegrees=15, twistLimitDegrees=15}
]
