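# Physics asset that simulates the front apron cloth of the female Sushi Ninja character, apparently for front-end display.
# Pelvis and thigh capsules form the body volumes, with the thighs and two small driver boxes angularly locked to the pelvis.
# Each driver anchors a left or right chain of three box bodies with limited swing and twist.
# Cross constraints tie the upper two segments of both chains together, leaving the bottom segments unlinked to each other.
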PhysicsAsset /BRCosmetics/Characters/Player/Female/Medium/Bodies/F_MED_Sushi_Ninja/Meshes/F_MED_Sushi_Ninja_Front_Cloth_FrontEnd_Physics
bodies {
 driver_apron_front_01_l [{shapeType=Box, extentX=1.01, extentY=1.01, extentZ=1.01, shapeName=driver_apron_front_01_l_box, rotation=<rot 0 -90 0>}]
 driver_apron_front_01_r [{shapeType=Box, extentX=1.01, extentY=1.01, extentZ=1.01, shapeName=driver_apron_front_01_r_box, rotation=<rot 0 -90 0>}]
 dyn_apron_frnt_rb_01_l [{shapeType=Box, extentX=10, extentY=2, extentZ=8, shapeName=dyn_apron_frnt_rb_01_l_box, center=<4 0 0>, rotation=<rot 90 0 0>}]
 dyn_apron_frnt_rb_01_r [{shapeType=Box, extentX=10, extentY=2, extentZ=8, shapeName=dyn_apron_frnt_rb_01_r_box, center=<4 0 0>, rotation=<rot 90 0 0>}]
 dyn_apron_frnt_rb_02_l [{shapeType=Box, extentX=10, extentY=3, extentZ=14, shapeName=dyn_apron_frnt_rb_02_l_box, center=<3 0 0>}]
 dyn_apron_frnt_rb_02_r [{shapeType=Box, extentX=10, extentY=3, extentZ=14, shapeName=dyn_apron_frnt_rb_02_r_box, center=<3 0 0>}]
 dyn_apron_frnt_rb_03_l [{shapeType=Box, extentX=17, extentY=3, extentZ=12, shapeName=dyn_apron_frnt_rb_03_l_box, center=<2 0 0>, rotation=<rot 90 0 0>}]
 dyn_apron_frnt_rb_03_r [{shapeType=Box, extentX=17, extentY=3, extentZ=12, shapeName=dyn_apron_frnt_rb_03_r_box, center=<2 0 0>, rotation=<rot 90 0 0>}]
 pelvis [{shapeType=Capsule, length=15, radius=8, shapeName=pelvis_capsule, center=<-2 0 0>}]
 thigh_l [{shapeType=Capsule, length=14, radius=8.5, shapeName=thigh_l_capsule, center=<-9 0 0>, rotation=<rot 90 0 0>}]
 thigh_r [{shapeType=Capsule, length=14, radius=8.5, shapeName=thigh_r_capsule, center=<9 0 -0.8>, rotation=<rot 90 0 0>}]
}
constraints [
 {
  bone1Name=thigh_l
  bone2Name=pelvis
  swing1LimitDegrees=45
  swing1Motion=Locked
  swing2LimitDegrees=45
  swing2Motion=Locked
  twistLimitDegrees=45
  twistMotion=Locked
 }
 {
  bone1Name=thigh_r
  bone2Name=pelvis
  swing1LimitDegrees=45
  swing1Motion=Locked
  swing2LimitDegrees=45
  swing2Motion=Locked
  twistLimitDegrees=45
  twistMotion=Locked
 }
 {
  bone1Name=driver_apron_front_01_l
  bone2Name=pelvis
  swing1LimitDegrees=45
  swing1Motion=Locked
  swing2LimitDegrees=45
  swing2Motion=Locked
  twistLimitDegrees=45
  twistMotion=Locked
 }
 {bone1Name=dyn_apron_frnt_rb_01_l, bone2Name=driver_apron_front_01_l, swing1LimitDegrees=10, swing2LimitDegrees=20, twistLimitDegrees=20}
 {bone1Name=dyn_apron_frnt_rb_02_l, bone2Name=dyn_apron_frnt_rb_01_l, swing1LimitDegrees=20, swing2LimitDegrees=20, twistLimitDegrees=30}
 {bone1Name=dyn_apron_frnt_rb_03_l, bone2Name=dyn_apron_frnt_rb_02_l, swing1LimitDegrees=20, swing2LimitDegrees=15, twistLimitDegrees=45}
 {
  bone1Name=driver_apron_front_01_r
  bone2Name=pelvis
  swing1LimitDegrees=45
  swing1Motion=Locked
  swing2LimitDegrees=45
  swing2Motion=Locked
  twistLimitDegrees=45
  twistMotion=Locked
 }
 {bone1Name=dyn_apron_frnt_rb_01_r, bone2Name=driver_apron_front_01_r, swing1LimitDegrees=10, swing2LimitDegrees=20, twistLimitDegrees=20}
 {bone1Name=dyn_apron_frnt_rb_02_r, bone2Name=dyn_apron_frnt_rb_01_r, swing1LimitDegrees=20, swing2LimitDegrees=20, twistLimitDegrees=30}
 {bone1Name=dyn_apron_frnt_rb_03_r, bone2Name=dyn_apron_frnt_rb_02_r, swing1LimitDegrees=20, swing2LimitDegrees=15, twistLimitDegrees=45}
 {bone1Name=dyn_apron_frnt_rb_01_r, bone2Name=dyn_apron_frnt_rb_01_l, swing1LimitDegrees=45, swing2LimitDegrees=45, twistLimitDegrees=45}
 {bone1Name=dyn_apron_frnt_rb_02_r, bone2Name=dyn_apron_frnt_rb_02_l, swing1LimitDegrees=45, swing2LimitDegrees=45, twistLimitDegrees=45}
]
