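# Physics asset for the left front apron cloth on the female medium Sushi Ninja body, named as a crouch variant.
# Three box bodies form a jointed chain hanging from a small driver sphere, which is angularly locked to the pelvis.
# Capsules on the pelvis and both thighs approximate the lower body, likely as collision for the apron, with both thighs angularly locked to the pelvis.
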PhysicsAsset /BRCosmetics/Characters/Player/Female/Medium/Bodies/F_MED_Sushi_Ninja/Meshes/F_MED_Sushi_Ninja_Front_Crouch_Cloth_L
bodies {
 driver_apron_front_01_l [{shapeType=Sphere, radius=0.505}]
 dyn_apron_frnt_rb_01_l [{shapeType=Box, extentX=10, extentY=2.5, extentZ=12.5, center=<4 0 0>}]
 dyn_apron_frnt_rb_02_l [{shapeType=Box, extentX=13, extentY=1.5, extentZ=16, center=<6.5 0 0.054>}]
 dyn_apron_frnt_rb_03_l [{shapeType=Box, extentX=10.5, extentY=3.2, extentZ=18, center=<6.19 0.2092 -0.6601>, rotation=<rot 0 2.134 0>}]
 pelvis [
  {shapeType=Capsule, length=15, radius=9, center=<-6.8762 0.9915 0>}
  {shapeType=Capsule, length=15, radius=9, shapeName=pelvis_capsule, center=<0 -0.7747 0>}
 ]
 thigh_l [{shapeType=Capsule, length=23, radius=9, center=<-14 0 0>, rotation=<rot 0 90 90>}]
 thigh_r [{shapeType=Capsule, length=23, radius=9, shapeName=thigh_r_capsule, center=<14 0 0>, rotation=<rot 0 90 90>}]
}
constraints [
 {
  bone1Name=driver_apron_front_01_l
  bone2Name=pelvis
  swing1LimitDegrees=45
  swing1Motion=Locked
  swing2LimitDegrees=45
  swing2Motion=Locked
  twistLimitDegrees=45
  twistMotion=Locked
 }
 {bone1Name=dyn_apron_frnt_rb_01_l, bone2Name=driver_apron_front_01_l, swing1LimitDegrees=45, swing2LimitDegrees=45, twistLimitDegrees=45}
 {bone1Name=dyn_apron_frnt_rb_02_l, bone2Name=dyn_apron_frnt_rb_01_l, swing1LimitDegrees=30, swing2LimitDegrees=30, twistLimitDegrees=30}
 {bone1Name=dyn_apron_frnt_rb_03_l, bone2Name=dyn_apron_frnt_rb_02_l, swing1LimitDegrees=30, swing2LimitDegrees=30, twistLimitDegrees=30}
 {
  bone1Name=thigh_l
  bone2Name=pelvis
  swing1LimitDegrees=45
  swing1Motion=Locked
  swing2LimitDegrees=45
  swing2Motion=Locked
  twistLimitDegrees=45
  twistMotion=Locked
 }
 {
  bone1Name=thigh_r
  bone2Name=pelvis
  swing1LimitDegrees=45
  swing1Motion=Locked
  swing2LimitDegrees=45
  swing2Motion=Locked
  twistLimitDegrees=45
  twistMotion=Locked
 }
]
