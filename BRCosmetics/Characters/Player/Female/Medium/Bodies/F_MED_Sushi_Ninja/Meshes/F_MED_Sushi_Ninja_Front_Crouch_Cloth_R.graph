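# Physics asset for the right front apron cloth on the medium female Sushi Ninja body, named as the crouch variant.
# A driver sphere, angularly locked to the pelvis, anchors a chain of three progressively larger box bodies with unlocked angular joints.
# Capsules on the pelvis and both thighs form the body shapes, with the thighs locked to the pelvis.
# A second, reversed constraint between the last two apron boxes locks their relative rotation.
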PhysicsAsset /BRCosmetics/Characters/Player/Female/Medium/Bodies/F_MED_Sushi_Ninja/Meshes/F_MED_Sushi_Ninja_Front_Crouch_Cloth_R
bodies {
 driver_apron_front_01_r [{shapeType=Sphere, radius=1}]
 dyn_apron_frnt_rb_01_r [{shapeType=Box, extentX=10, extentY=2, extentZ=13.5, center=<4 0 0>}]
 dyn_apron_frnt_rb_02_r [{shapeType=Box, extentX=11, extentY=2, extentZ=16, center=<4 0 0>}]
 dyn_apron_frnt_rb_03_r [{shapeType=Box, extentX=12, extentY=3, extentZ=17, center=<4 0 0>}]
 pelvis [
  {shapeType=Capsule, length=15, radius=9, center=<-4.9094 0.5817 0>}
  {shapeType=Capsule, length=15, radius=9, shapeName=pelvis_capsule, center=<0 -0.8294 0>}
 ]
 thigh_l [{shapeType=Capsule, length=23, radius=9, shapeName=thigh_l_capsule, center=<-14 0 0>, rotation=<rot 90 0 0>}]
 thigh_r [{shapeType=Capsule, length=26, radius=9, center=<14 0 0>, rotation=<rot 90 0 0>}]
}
constraints [
 {
  bone1Name=driver_apron_front_01_r
  bone2Name=pelvis
  swing1LimitDegrees=45
  swing1Motion=Locked
  swing2LimitDegrees=45
  swing2Motion=Locked
  twistLimitDegrees=45
  twistMotion=Locked
 }
 {bone1Name=dyn_apron_frnt_rb_01_r, bone2Name=driver_apron_front_01_r, swing1LimitDegrees=45, swing2LimitDegrees=45, twistLimitDegrees=45}
 {bone1Name=dyn_apron_frnt_rb_02_r, bone2Name=dyn_apron_frnt_rb_01_r, swing1LimitDegrees=45, swing2LimitDegrees=45, twistLimitDegrees=45}
 {bone1Name=dyn_apron_frnt_rb_03_r, bone2Name=dyn_apron_frnt_rb_02_r, swing1LimitDegrees=45, swing2LimitDegrees=45, twistLimitDegrees=45}
 {
  bone1Name=thigh_r
  bone2Name=pelvis
  swing1LimitDegrees=45
  swing1Motion=Locked
  swing2LimitDegrees=45
  swing2Motion=Locked
  twistLimitDegrees=45
  twistMotion=Locked
 }
 {
  bone1Name=dyn_apron_frnt_rb_02_r
  bone2Name=dyn_apron_frnt_rb_03_r
  swing1LimitDegrees=45
  swing1Motion=Locked
  swing2LimitDegrees=45
  swing2Motion=Locked
  twistLimitDegrees=45
  twistMotion=Locked
 }
 {
  bone1Name=thigh_l
  bone2Name=pelvis
  swing1LimitDegrees=45
  swing1Motion=Locked
  swing2LimitDegrees=45
  swing2Motion=Locked
  twistLimitDegrees=45
  twistMotion=Locked
 }
]
