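# Physics asset for the knives and rear cloth on the female medium Sushi Ninja character body.
# Pelvis and thigh capsules form the body volume, with tiny anchor bodies locked to the pelvis for each knife and the cloth.
# Five capsule blades hang from their anchors with narrow angular limits.
# The rear cloth is a two-segment middle chain that branches into two side bodies, one named with the misspelling "dyn_croth_rear_reft".
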PhysicsAsset /BRCosmetics/Characters/Player/Female/Medium/Bodies/F_MED_Sushi_Ninja/Meshes/F_MED_Sushi_Ninja_Knives_Physics2
bodies {
 base_cloth_rear_mid [{shapeType=Capsule, length=0.505, radius=0.505, shapeName=base_cloth_rear_mid_capsule, rotation=<rot 0 -90 0>}]
 dyn_bladeA_02_r [{shapeType=Capsule, length=17, radius=2, shapeName=dyn_bladeA_02_r_capsule, center=<11 0 0>, rotation=<rot 85.0001 179.9996 179.9995>}]
 dyn_bladeB_01_l [{shapeType=Capsule, length=17.5, radius=1.5, shapeName=dyn_bladeB_01_l_capsule, center=<10.9727 -0.4107 0>, rotation=<rot 90 -180 -180>}]
 dyn_bladeB_01_r [{shapeType=Sphere, radius=0.505}]
 dyn_bladeB_02_r [{shapeType=Capsule, length=17, radius=2, shapeName=dyn_bladeB_02_r_capsule, center=<11 0 0>, rotation=<rot -90 -157.1625 157.1626>}]
 dyn_bladeC_01_l [{shapeType=Sphere, radius=0.505}]
 dyn_bladeC_02_l [{shapeType=Capsule, length=17.5, radius=1.5, shapeName=dyn_bladeC_02_l_capsule, center=<11.0255 -0.4171 0>, rotation=<rot -90 180 -176.4989>}]
 dyn_bladeD_01_l [{shapeType=Sphere, radius=0.505}]
 dyn_bladeD_02_l [{
   shapeType=Capsule
   length=19
   radius=1.5
   shapeName=dyn_bladeD_02_l_capsule
   center=<11.3308 -1.0948 -0.0217>
   rotation=<rot -90 179.9978 -179.9973>
  }]
 dyn_cloth_rear_left [{shapeType=Capsule, length=3, radius=3, shapeName=dyn_cloth_rear_left_capsule, center=<5 -1.8 0>}]
 dyn_cloth_rear_mid_01 [{shapeType=Capsule, length=10, radius=1.9, shapeName=dyn_cloth_rear_mid_01_capsule, center=<0 -0.4 -1.5>, rotation=<rot 90 0 0>}]
 dyn_cloth_rear_mid_02 [
  {shapeType=Capsule, length=14.5, radius=1.3, shapeName=dyn_cloth_rear_mid_02_capsule, center=<0 -0.8 -0.7>, rotation=<rot -90 0 0>}
  {shapeType=Capsule, length=14.5, radius=3, shapeName=dyn_cloth_rear_mid_02_capsule, center=<0 0 -8>, rotation=<rot -90 0 0>}
 ]
 dyn_croth_rear_reft [{shapeType=Capsule, length=3, radius=3, shapeName=dyn_croth_rear_reft_capsule, center=<-5 1.8 0>}]
 dyn_driver_bladeA_r [{shapeType=Sphere, radius=0.505}]
 dyn_driver_bladeB_l [{shapeType=Sphere, radius=0.505}]
 pelvis [{shapeType=Capsule, length=14.2, radius=11, shapeName=pelvis_capsule, center=<-4.3 -0.8613 0>}]
 thigh_l [{shapeType=Capsule, length=12, radius=11, center=<-14 -2 1.5>, rotation=<rot 90 0 0>}]
 thigh_r [{shapeType=Capsule, length=12, radius=11, center=<14 2 -1.5>, rotation=<rot 90 180 0>}]
}
constraints [
 {bone1Name=dyn_bladeC_02_l, bone2Name=dyn_bladeC_01_l, swing1LimitDegrees=15, swing2LimitDegrees=20, twistLimitDegrees=8}
 {bone1Name=dyn_bladeD_02_l, bone2Name=dyn_bladeD_01_l, swing1LimitDegrees=20, swing2LimitDegrees=18, twistLimitDegrees=8}
 {bone1Name=dyn_bladeB_02_r, bone2Name=dyn_bladeB_01_r, swing1LimitDegrees=10, swing2LimitDegrees=25, twistLimitDegrees=8}
 {bone1Name=dyn_bladeB_01_l, bone2Name=dyn_driver_bladeB_l, swing1LimitDegrees=15, swing2LimitDegrees=20, twistLimitDegrees=8}
 {bone1Name=dyn_bladeA_02_r, bone2Name=dyn_driver_bladeA_r, swing1LimitDegrees=15, swing2LimitDegrees=20, twistLimitDegrees=8}
 {
  bone1Name=thigh_l
  bone2Name=pelvis
  swing1LimitDegrees=45
  swing1Motion=Locked
  swing2LimitDegrees=45
  swing2Motion=Locked
  twistLimitDegrees=45
  twistMotion=Locked
 }
 {
  bone1Name=dyn_driver_bladeB_l
  bone2Name=pelvis
  swing1LimitDegrees=45
  swing1Motion=Locked
  swing2LimitDegrees=45
  swing2Motion=Locked
  twistLimitDegrees=45
  twistMotion=Locked
 }
 {
  bone1Name=dyn_bladeC_01_l
  bone2Name=pelvis
  swing1LimitDegrees=45
  swing1Motion=Locked
  swing2LimitDegrees=45
  swing2Motion=Locked
  twistLimitDegrees=45
  twistMotion=Locked
 }
 {
  bone1Name=dyn_bladeD_01_l
  bone2Name=pelvis
  swing1LimitDegrees=45
  swing1Motion=Locked
  swing2LimitDegrees=45
  swing2Motion=Locked
  twistLimitDegrees=45
  twistMotion=Locked
 }
 {
  bone1Name=dyn_driver_bladeA_r
  bone2Name=pelvis
  swing1LimitDegrees=45
  swing1Motion=Locked
  swing2LimitDegrees=45
  swing2Motion=Locked
  twistLimitDegrees=45
  twistMotion=Locked
 }
 {
  bone1Name=dyn_bladeB_01_r
  bone2Name=pelvis
  swing1LimitDegrees=45
  swing1Motion=Locked
  swing2LimitDegrees=45
  swing2Motion=Locked
  twistLimitDegrees=45
  twistMotion=Locked
 }
 {bone1Name=dyn_cloth_rear_mid_02, bone2Name=dyn_cloth_rear_mid_01, swing1LimitDegrees=15, swing2LimitDegrees=25, twistLimitDegrees=0}
 {bone1Name=dyn_cloth_rear_mid_01, bone2Name=base_cloth_rear_mid, swing1LimitDegrees=25, swing2LimitDegrees=20, twistLimitDegrees=0}
 {bone1Name=dyn_cloth_rear_left, bone2Name=dyn_cloth_rear_mid_02, swing1LimitDegrees=20, swing2LimitDegrees=30, twistLimitDegrees=30}
 {bone1Name=dyn_croth_rear_reft, bone2Name=dyn_cloth_rear_mid_02, swing1LimitDegrees=20, swing2LimitDegrees=30, twistLimitDegrees=30}
 {
  bone1Name=base_cloth_rear_mid
  bone2Name=pelvis
  swing1LimitDegrees=45
  swing1Motion=Locked
  swing2LimitDegrees=45
  swing2Motion=Locked
  twistLimitDegrees=45
  twistMotion=Locked
 }
 {
  bone1Name=thigh_r
  bone2Name=pelvis
  swing1LimitDegrees=45
  swing1Motion=Locked
  swing2LimitDegrees=45
  swing2Motion=Locked
  twistLimitDegrees=45
  twistMotion=Locked
 }
]
