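# Post-process style AnimBlueprint for the Sushi Ninja cosmetic body, adding secondary motion to apron, knives, bows and elbows.
# Copied parent pose runs through Control Rig, elbow pose drivers and AnimDynamics into a cache.
# Thigh-driven flap nodes feed standing or crouch/DBNO RigidBody apron sims, with a non-simulated flap fallback when rigid bodies are disabled.
# Event graph tracks movement states, derives thigh-based values and overrides cloth gravity while skydiving.
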AnimBlueprint /BRCosmetics/Characters/Player/Female/Medium/Bodies/F_MED_Sushi_Ninja/Meshes/F_MED_Sushi_Ninja_AnimBP
target_skeleton=&/F_MED_Sushi_Ninja_Skeleton
graphs [
 AnimGraph {nodes [
   n1 = "Constraint - dyn_apron_frnt_rb_01_l (RR)"!(
    ComponentPose=Local To Component(LocalPose=Use cached pose 'PreSim_Cache'())
    Alpha=1
    ConstraintWeights_0=0.93
    ConstraintWeights_1:float=0.07
   ) -> (Pose) {
    alphaBoolBlend {blendOption=Linear}
    boneToModify {boneName=dyn_apron_frnt_rb_01_l}
    constraintSetup [{transformType=Rotation, targetBone {boneName=thigh_l}}, {transformType=Rotation, targetBone {boneName=pelvis}}]
    constraintWeights [1,1]
   }
   n2 = Flap Node - Bone: dyn_apron_frnt_rb_01_r!(ComponentPose:Component Space Pose Link Structure=@n1, Alpha:float=1) -> (Pose:Component Space Pose Link Structure) {
    is=title FortAnimGraphNode_Flap
    bUseClamp=true
    targetClampMax=0.1
    targetClampMin=-3.14
    targetComponent=RotationZ
    alphaBoolBlend {blendOption=Linear}
    sourceDrives [{
      bUseRange=false
      drivingCurve=null
      multiplier=-1
      rangeMax=1
      rangeMin=-1
      remappedMax=1
      remappedMin=0
      sourceComponent=RotationZ
      useQuaternion=true
      sourceBone {boneName=thigh_r}
     }]
    targetBone {boneName=dyn_apron_frnt_rb_01_r}
   }
   n3 = Flap Node - Bone: dyn_apron_frnt_rb_01_l!(ComponentPose:Component Space Pose Link Structure=@n2, Alpha:float=1) -> (Pose:Component Space Pose Link Structure) {
    is=title FortAnimGraphNode_Flap
    bUseClamp=true
    targetClampMax=0.1
    targetClampMin=-1.5
    targetComponent=RotationZ
    alphaBoolBlend {blendOption=Linear}
    sourceDrives [{
      bUseRange=false
      drivingCurve=null
      multiplier=-1
      rangeMax=1
      rangeMin=-1
      remappedMax=1
      remappedMin=0
      sourceComponent=RotationZ
      useQuaternion=true
      sourceBone {boneName=thigh_l}
     }]
    targetBone {boneName=dyn_apron_frnt_rb_01_l}
   }
   n4 = Component To Local(ComponentPose=@n3)
   n5 = Get bRigidBodyEnabled()
   n6 = Flap Node - Bone: driver_apron_front_01_r!(ComponentPose:Component Space Pose Link Structure=Local To Component(LocalPose=Use cached pose 'PreSim_Cache'()), Alpha:float=0.5) -> (Pose:Component Space Pose Link Structure) {
    is=title FortAnimGraphNode_Flap
    bUseClamp=true
    targetClampMin=-10
    targetComponent=RotationZ
    alphaBoolBlend {blendOption=Linear}
    sourceDrives [{
      bUseRange=false
      drivingCurve=null
      multiplier=-1
      rangeMax=1
      rangeMin=-1
      remappedMax=1
      remappedMin=0
      sourceComponent=RotationZ
      useQuaternion=true
      sourceBone {boneName=thigh_r}
     }]
    targetBone {boneName=driver_apron_front_01_r}
   }
   n7 = Flap Node - Bone: driver_apron_front_01_l!(ComponentPose:Component Space Pose Link Structure=@n6, Alpha:float=0.5) -> (Pose:Component Space Pose Link Structure) {
    is=title FortAnimGraphNode_Flap
    bUseClamp=true
    targetClampMin=-10
    targetComponent=RotationZ
    alphaBoolBlend {blendOption=Linear}
    sourceDrives [{
      bUseRange=false
      drivingCurve=null
      multiplier=-1
      rangeMax=1
      rangeMin=-1
      remappedMax=1
      remappedMin=0
      sourceComponent=RotationZ
      useQuaternion=true
      sourceBone {boneName=thigh_l}
     }]
    targetBone {boneName=driver_apron_front_01_l}
   }
   n8 = Break Fort Anim Input Skydiving External Force(FortAnimInput_SkydivingExternalForce=Get SkydivingExternalForceInput()) -> (
    bUseSkydivingVectorForce=~
    SkydivingVectorForce=~
    HeadToPelvisDirection=~
    FloatingMultiplier=~
    FloatingAdditive=~
    DivingMultiplier=~
    DivingAdditive=~
    ParachutingMultiplier=~
    ParachutingAdditive=~
   )
   n9 = RigidBody(ComponentPose=@n7, ExternalForce=@n8.SkydivingVectorForce) {
    bClampLinearTranslationLimitToRefPose=true
    bOverrideWorldGravity=true
    overridePhysicsAsset=&/F_MED_Sushi_Ninja_Front_Cloth
    simulationSpace=BaseBoneSpace
    alphaBoolBlend {blendOption=Linear}
    baseBoneRef {boneName=pelvis}
    componentAppliedLinearAccClamp=<3000 1000 3000>
    componentLinearAccScale=<0.3 0.3 0.3>
    componentLinearVelScale=<0.3 0.3 0.3>
    overrideWorldGravity {z=-980}
    simSpaceSettings {worldAlpha=0.2}
   }
   n10 = "Transform (Modify) Bone - Bone: driver_apron_front_01_r"!(ComponentPose=@n9, bAlphaBoolEnabled:bool=True, Translation="0,0,0", Rotation="0,-15,0", Scale="1,1,1") -> (Pose) {
    alphaInputType=Bool
    rotationMode=BMM_Additive
    rotationSpace=BCS_BoneSpace
    alphaBoolBlend {blendOption=Linear}
    alphaScaleBiasClamp {bInterpResult=true}
    boneToModify {boneName=driver_apron_front_01_r}
    rotation {yaw=-15}
   }
   n11 = "Transform (Modify) Bone - Bone: driver_apron_front_01_l"!(ComponentPose=@n10, bAlphaBoolEnabled:bool=True, Translation="0,0,0", Rotation="0,-10,0", Scale="1,1,1") -> (Pose) {
    alphaInputType=Bool
    rotationMode=BMM_Additive
    rotationSpace=BCS_BoneSpace
    alphaBoolBlend {blendInTime=0.2, blendOption=Linear, blendOutTime=1}
    alphaScaleBiasClamp {bInterpResult=true}
    boneToModify {boneName=driver_apron_front_01_l}
    rotation {yaw=-10}
   }
   n12 = "Transform (Modify) Bone - Bone: driver_apron_front_01_r"!(ComponentPose=@n11, Alpha:float=1, Translation="0,0,0", Rotation="-0.730,-3.829,-0.700", Scale="1,1,1") -> (Pose) {
    rotationMode=BMM_Additive
    rotationSpace=BCS_BoneSpace
    alphaBoolBlend {blendInTime=0.2, blendOption=Linear, blendOutTime=1}
    boneToModify {boneName=driver_apron_front_01_r}
    rotation=<rot -0.73 -3.829 -0.7>
   }
   n13 = "Transform (Modify) Bone - Bone: dyn_apron_frnt_rb_02_r"(ComponentPose=@n12, Alpha=1, Translation="0,0,0", Rotation="0,-7,0", Scale="1,1,1") {
    rotationMode=BMM_Additive
    rotationSpace=BCS_BoneSpace
    alphaBoolBlend {blendInTime=0.2, blendOption=Linear, blendOutTime=1}
    boneToModify {boneName=dyn_apron_frnt_rb_02_r}
    rotation {yaw=-7}
   }
   n14 = "Transform (Modify) Bone - Bone: driver_apron_front_01_l"!(ComponentPose=@n13, Alpha:float=1, Translation="0,0,0", Rotation="0,-15,0", Scale="1,1,1") -> (Pose) {
    rotationMode=BMM_Additive
    rotationSpace=BCS_BoneSpace
    alphaBoolBlend {blendInTime=0.2, blendOption=Linear, blendOutTime=1}
    boneToModify {boneName=driver_apron_front_01_l}
    rotation {yaw=-15}
   }
   n15 = "Transform (Modify) Bone - Bone: dyn_apron_frnt_rb_02_l"(ComponentPose=@n14, Alpha=1, Translation="0,0,0", Rotation="0,-15,0", Scale="1,1,1") {
    rotationMode=BMM_Additive
    rotationSpace=BCS_BoneSpace
    alphaBoolBlend {blendInTime=0.2, blendOption=Linear, blendOutTime=1}
    boneToModify {boneName=dyn_apron_frnt_rb_02_l}
    rotation {yaw=-15}
   }
   n16 = Component To Local(ComponentPose=@n15)
   n17 = Flap Node - Bone: driver_apron_front_01_r(ComponentPose=Local To Component(LocalPose=Use cached pose 'PreSim_Cache'()), Alpha=0.9) {
    bUseClamp=true
    targetClampMin=-10
    targetComponent=RotationZ
    alphaBoolBlend {blendOption=Linear}
    sourceDrives [{
      bUseRange=false
      drivingCurve=null
      multiplier=-1
      rangeMax=1
      rangeMin=-1
      remappedMax=1
      remappedMin=0
      sourceComponent=RotationZ
      useQuaternion=true
      sourceBone {boneName=thigh_r}
     }]
    targetBone {boneName=driver_apron_front_01_r}
   }
   n18 = Flap Node - Bone: driver_apron_front_01_l(ComponentPose=@n17, Alpha=0.9) {
    bUseClamp=true
    targetClampMin=-10
    targetComponent=RotationZ
    alphaBoolBlend {blendOption=Linear}
    sourceDrives [{
      bUseRange=false
      drivingCurve=null
      multiplier=-1
      rangeMax=1
      rangeMin=-1
      remappedMax=1
      remappedMin=0
      sourceComponent=RotationZ
      useQuaternion=true
      sourceBone {boneName=thigh_l}
     }]
    targetBone {boneName=driver_apron_front_01_l}
   }
   n19 = RigidBody(ComponentPose=@n18, ExternalForce=@n8.SkydivingVectorForce) {
    bClampLinearTranslationLimitToRefPose=true
    bOverrideWorldGravity=true
    overridePhysicsAsset=&/F_MED_Sushi_Ninja_Front_Crouch_Cloth_L
    simulationSpace=BaseBoneSpace
    alphaBoolBlend {blendOption=Linear}
    alphaScaleBiasClamp {bInterpResult=true}
    baseBoneRef {boneName=pelvis}
    componentAppliedLinearAccClamp=<1000 1000 1000>
    componentLinearAccScale=<1 1 1>
    componentLinearVelScale=<3 3 1>
    overrideWorldGravity {z=-1000}
   }
   n20 = RigidBody(ComponentPose=@n19, ExternalForce=@n8.SkydivingVectorForce) {
    bClampLinearTranslationLimitToRefPose=true
    bOverrideWorldGravity=true
    overridePhysicsAsset=&/F_MED_Sushi_Ninja_Front_Crouch_Cloth_R
    simulationSpace=BaseBoneSpace
    alphaBoolBlend {blendOption=Linear}
    alphaScaleBiasClamp {bInterpResult=true}
    baseBoneRef {boneName=pelvis}
    componentAppliedLinearAccClamp=<1000 1000 1000>
    componentLinearAccScale=<1 1 1>
    componentLinearVelScale=<3 3 1>
    overrideWorldGravity {z=-1000}
   }
   n21 = "Transform (Modify) Bone - Bone: driver_apron_front_01_l"!(ComponentPose=@n20, bAlphaBoolEnabled:bool=True, Translation="0,-3,0", Rotation="0,0,0", Scale="1,1,1") -> (Pose) {
    alphaInputType=Bool
    translationMode=BMM_Additive
    translationSpace=BCS_BoneSpace
    alphaBoolBlend {blendInTime=0.4, blendOption=Linear, blendOutTime=1}
    boneToModify {boneName=driver_apron_front_01_l}
    translation {y=-3}
   }
   n22 = "Transform (Modify) Bone - Bone: driver_apron_front_01_r"!(ComponentPose=@n21, bAlphaBoolEnabled:bool=True, Translation="0,-3,0", Rotation="0,0,0", Scale="1,1,1") -> (Pose) {
    alphaInputType=Bool
    translationMode=BMM_Additive
    translationSpace=BCS_BoneSpace
    alphaBoolBlend {blendInTime=0.4, blendOption=Linear, blendOutTime=1}
    boneToModify {boneName=driver_apron_front_01_r}
    translation {y=-3}
   }
   n23 = "Transform (Modify) Bone - Bone: driver_apron_front_01_l"!(ComponentPose=@n22, Alpha:float=1, Translation="0,-2,0", Rotation="0,0,0", Scale="1,1,1") -> (Pose) {
    translationMode=BMM_Additive
    translationSpace=BCS_BoneSpace
    alphaBoolBlend {blendOption=Linear}
    boneToModify {boneName=driver_apron_front_01_l}
    translation {y=-2}
   }
   n24 = "Transform (Modify) Bone - Bone: dyn_apron_frnt_rb_03_r"!(
    ComponentPose=@n23
    bAlphaBoolEnabled:bool=NOT Boolean(A=Get Moving2D() -> (Moving2D:bool))
    Translation="0,0,0"
    Rotation="0,-30,0"
    Scale="1,1,1"
   ) -> (Pose) {
    alphaInputType=Bool
    rotationMode=BMM_Additive
    translationSpace=BCS_BoneSpace
    alphaBoolBlend {blendInTime=0.2, blendOption=Linear, blendOutTime=1}
    boneToModify {boneName=dyn_apron_frnt_rb_03_r}
    rotation {yaw=-30}
   }
   n25 = "Transform (Modify) Bone - Bone: driver_apron_front_01_r"(ComponentPose=@n24, bAlphaBoolEnabled=True, Translation="0,0,0", Rotation="0,-25,0", Scale="1,1,1") {
    alphaInputType=Bool
    rotationMode=BMM_Additive
    rotationSpace=BCS_BoneSpace
    alphaBoolBlend {blendInTime=0.2, blendOption=Linear, blendOutTime=1}
    alphaScaleBiasClamp {bInterpResult=true}
    boneToModify {boneName=driver_apron_front_01_r}
    rotation {yaw=-25}
   }
   n26 = "Transform (Modify) Bone - Bone: driver_apron_front_01_l"!(ComponentPose=@n25, bAlphaBoolEnabled:bool=True, Translation="0,0,0", Rotation="0,-5,0", Scale="1,1,1") -> (Pose) {
    alphaInputType=Bool
    rotationMode=BMM_Additive
    rotationSpace=BCS_BoneSpace
    alphaBoolBlend {blendInTime=0.2, blendOption=Linear, blendOutTime=1}
    alphaScaleBiasClamp {bInterpResult=true}
    boneToModify {boneName=driver_apron_front_01_l}
    rotation {yaw=-5}
   }
   n27 = Component To Local(ComponentPose=@n26)
   n28 = Get CrouchingDynamics() -> (CrouchingDynamics:bool)
   n29 = RigidBody(ComponentPose=Local To Component(LocalPose=Blend Poses by bool(BlendPose_0=@n27, BlendPose_1=@n16, BlendTime_0=0.5, BlendTime_1=0.5, bActiveValue=@n28) {blendType=QuadraticInOut})) {
    bClampLinearTranslationLimitToRefPose=true
    bOverrideWorldGravity=true
    overridePhysicsAsset=&/F_MED_Sushi_Ninja_Knives_Physics2
    simulationSpace=BaseBoneSpace
    alphaBoolBlend {blendOption=Linear}
    alphaScaleBiasClamp {bInterpResult=true, interpSpeedDecreasing=1, interpSpeedIncreasing=1}
    baseBoneRef {boneName=pelvis}
    componentAppliedLinearAccClamp=<3000 3000 1000>
    componentLinearAccScale=<0.7 0.6 0.6>
    componentLinearVelScale=<0.7 0.6 0.6>
    overrideWorldGravity {z=-920}
   }
   n30 = Anim Dynamics - Bone: dyn_bow_top_01_l(ComponentPose=@n29, Alpha=1, ExternalForce="0,0,0") {
    bUseGravityOverride=true
    simulationSpace=BoneRelative
    alphaBoolBlend {blendOption=Linear}
    boundBone {boneName=dyn_bow_top_01_l}
    componentAppliedLinearAccClamp=<3000 3000 3000>
    gravityOverride {z=-980}
    physicsBodyDefinitions [{
      boundBone {boneName=dyn_bow_top_01_l}
      boxExtents=<2 2 4>
      constraintSetup {angularLimitsMax=<0 3 3>, angularLimitsMin=<0 -3 -3>, angularTarget=<0 0 0>}
      localJointOffset=<1.5 0 0>
     }]
    relativeSpaceBone {boneName=root}
   }
   n31 = Anim Dynamics - Bone: dyn_bow_top_01_r(ComponentPose=@n30, Alpha=1, ExternalForce="0,0,0") {
    bUseGravityOverride=true
    simulationSpace=BoneRelative
    alphaBoolBlend {blendOption=Linear}
    boundBone {boneName=dyn_bow_top_01_r}
    componentAppliedLinearAccClamp=<3000 3000 3000>
    gravityOverride {z=-980}
    physicsBodyDefinitions [{
      boundBone {boneName=dyn_bow_top_01_r}
      boxExtents=<2 2 4>
      constraintSetup {angularLimitsMax=<0 3 3>, angularLimitsMin=<0 -3 -3>, angularTarget=<0 0 0>}
      localJointOffset=<-1.5 0 0>
     }]
    relativeSpaceBone {boneName=root}
   }
   n32 = Component To Local(ComponentPose=@n31)
   n33 = RigidBody!(
    ComponentPose=Local To Component(LocalPose=Blend Poses by bool(BlendPose_0=@n32, BlendPose_1=@n4, bActiveValue=@n5) {blendType=Linear})
    bAlphaBoolEnabled:bool
    ExternalForce
   ) -> (Pose) {
    alphaInputType=Bool
    bClampLinearTranslationLimitToRefPose=true
    bOverrideWorldGravity=true
    overridePhysicsAsset=&/F_MED_Sushi_Ninja_Front_Cloth_FrontEnd_Physics
    simulationSpace=BaseBoneSpace
    alphaBoolBlend {blendInTime=0.2, blendOption=Linear, blendOutTime=1}
    baseBoneRef {boneName=pelvis}
    componentAppliedLinearAccClamp=<3000 1000 3000>
    componentLinearAccScale=<0.3 0.3 0.3>
    componentLinearVelScale=<0.3 0.3 0.3>
    overrideWorldGravity {z=-980}
    simSpaceSettings {worldAlpha=0.2}
   }
   Output Pose(Result=Component To Local(ComponentPose=@n33)) {layerGroup=None}
   n34 = Control Rig(Source=Copy Pose From Mesh() {bCopyCurves=true, bUseAttachedParent=true}) {
    alphaBoolBlend {blendOption=Linear}
    controlRigAssetReference {blueprintRigClass=&/F_MED_Sushi_Ninja_CtrlRig.F_MED_Sushi_Ninja_CtrlRig_C}
   }
   n35 = Pose Driver - Source: lowerarm_l!(SourcePose:Pose Link Structure=@n34) -> (Pose:Pose Link Structure) {
    is=title AnimGraphNode_PoseDriver
    poseAsset=/BRCosmetics/Characters/Player/Female/Medium/Bodies/F_MED_SushiChef/Meshes/F_MED_Sushi_Chef_Skeleton_PoseAsset
    onlyDriveBones [{boneName=driver_elbow_l}, {boneName=elbow_inside_l}, {boneName=elbow_inner_l}, {boneName=elbow_outer_l}, {boneName=driver_dyn_sleeve_l}]
    poseTargets [
     {
      bApplyCustomCurve=true
      bIsHidden=false
      distanceMethod=DefaultMethod
      drivenName=Pose_0
      functionType=DefaultFunction
      targetScale=1
      boneTransforms [{targetRotation=<rot 0 0 0>, targetTranslation=<0 0 0>}, {targetRotation=<rot 0 0 0>, targetTranslation=<0 0 0>}]
      customCurve {
       defaultValue=3.4028234663852886e+38
       postInfinityExtrap=RCCE_Constant
       preInfinityExtrap=RCCE_Constant
       keys [
        {
         arriveTangent=0
         arriveTangentWeight=0
         interpMode=RCIM_Linear
         leaveTangent=0
         leaveTangentWeight=0
         tangentMode=RCTM_Auto
         tangentWeightMode=RCTWM_WeightedNone
         time=0
         value=0
        }
        {
         arriveTangent=0
         arriveTangentWeight=0
         interpMode=RCIM_Linear
         leaveTangent=0
         leaveTangentWeight=0
         tangentMode=RCTM_Auto
         tangentWeightMode=RCTWM_WeightedNone
         time=1
         value=1
        }
       ]
      }
      targetRotation=<rot 0 0 0>
     }
     {
      bApplyCustomCurve=false
      bIsHidden=false
      distanceMethod=DefaultMethod
      drivenName=Pose_1
      functionType=DefaultFunction
      targetScale=1
      boneTransforms [{targetRotation=<rot 0 0 0>, targetTranslation=<0 0 0>}, {targetRotation=<rot 0 0 0>, targetTranslation=<0 0 0>}]
      customCurve {defaultValue=3.4028234663852886e+38, postInfinityExtrap=RCCE_Constant, preInfinityExtrap=RCCE_Constant, keys []}
      targetRotation=<rot 0 0 0>
     }
     {
      bApplyCustomCurve=false
      bIsHidden=false
      distanceMethod=DefaultMethod
      drivenName=Pose_2
      functionType=DefaultFunction
      targetScale=1
      boneTransforms [{targetRotation=<rot 0 0 0>, targetTranslation=<0 0 0>}, {targetRotation=<rot 0 0 0>, targetTranslation=<0 0 0>}]
      customCurve {defaultValue=3.4028234663852886e+38, postInfinityExtrap=RCCE_Constant, preInfinityExtrap=RCCE_Constant, keys []}
      targetRotation=<rot 0 0 0>
     }
     {
      bApplyCustomCurve=false
      bIsHidden=false
      distanceMethod=DefaultMethod
      drivenName=Pose_3
      functionType=DefaultFunction
      targetScale=1
      boneTransforms [{targetRotation=<rot 0 0 0>, targetTranslation=<0 0 0>}, {targetRotation=<rot 0 0 0>, targetTranslation=<0 0 0>}]
      customCurve {defaultValue=3.4028234663852886e+38, postInfinityExtrap=RCCE_Constant, preInfinityExtrap=RCCE_Constant, keys []}
      targetRotation=<rot 0 0 0>
     }
     {
      bApplyCustomCurve=false
      bIsHidden=false
      distanceMethod=DefaultMethod
      drivenName=Pose_4
      functionType=DefaultFunction
      targetScale=1
      boneTransforms [{targetRotation=<rot 0 0 0>, targetTranslation=<0 0 0>}, {targetRotation=<rot 0 0 0>, targetTranslation=<0 0 0>}]
      customCurve {defaultValue=3.4028234663852886e+38, postInfinityExtrap=RCCE_Constant, preInfinityExtrap=RCCE_Constant, keys []}
      targetRotation=<rot 0 0 0>
     }
     {
      bApplyCustomCurve=false
      bIsHidden=false
      distanceMethod=DefaultMethod
      drivenName=Pose_5
      functionType=DefaultFunction
      targetScale=1
      boneTransforms [{targetRotation=<rot 0 0 0>, targetTranslation=<0 0 0>}, {targetRotation=<rot 0 0 0>, targetTranslation=<0 0 0>}]
      customCurve {defaultValue=3.4028234663852886e+38, postInfinityExtrap=RCCE_Constant, preInfinityExtrap=RCCE_Constant, keys []}
      targetRotation=<rot 0 0 0>
     }
     {
      bApplyCustomCurve=false
      bIsHidden=false
      distanceMethod=DefaultMethod
      drivenName=Pose_6
      functionType=DefaultFunction
      targetScale=1
      boneTransforms [{targetRotation=<rot 0 0 0>, targetTranslation=<0 0 0>}, {targetRotation=<rot 0 0 0>, targetTranslation=<0 0 0>}]
      customCurve {defaultValue=3.4028234663852886e+38, postInfinityExtrap=RCCE_Constant, preInfinityExtrap=RCCE_Constant, keys []}
      targetRotation=<rot 0 0 0>
     }
     {
      bApplyCustomCurve=false
      bIsHidden=false
      distanceMethod=DefaultMethod
      drivenName=Pose_7
      functionType=DefaultFunction
      targetScale=1
      boneTransforms [{targetRotation=<rot 0 0 0>, targetTranslation=<0 0 0>}, {targetRotation=<rot 0 0 0>, targetTranslation=<0 0 0>}]
      customCurve {defaultValue=3.4028234663852886e+38, postInfinityExtrap=RCCE_Constant, preInfinityExtrap=RCCE_Constant, keys []}
      targetRotation=<rot 0 0 0>
     }
    ]
    rBFParams {distanceMethod=Quaternion, radius=55}
    sourceBones [{boneName=lowerarm_l}]
   }
   n36 = Pose Driver - Source: lowerarm_r!(SourcePose:Pose Link Structure=@n35) -> (Pose:Pose Link Structure) {
    is=title AnimGraphNode_PoseDriver
    poseAsset=&F_MED_Sushi_Chef_Skeleton_PoseAsset
    onlyDriveBones [{boneName=driver_elbow_r}, {boneName=elbow_inside_r}, {boneName=elbow_inner_r}, {boneName=elbow_outer_r}, {boneName=driver_dyn_sleeve_r}]
    poseTargets [
     {
      bApplyCustomCurve=false
      bIsHidden=false
      distanceMethod=DefaultMethod
      drivenName=Pose_0
      functionType=DefaultFunction
      targetScale=1
      boneTransforms [{targetRotation=<rot 1.7008 26.5846 0.8306>, targetTranslation=<-26.4651 0.8345 -0.259>}]
      customCurve {defaultValue=3.4028234663852886e+38, postInfinityExtrap=RCCE_Constant, preInfinityExtrap=RCCE_Constant, keys []}
      targetRotation=<rot 0 0 0>
     }
     {
      bApplyCustomCurve=false
      bIsHidden=false
      distanceMethod=DefaultMethod
      drivenName=Pose_1
      functionType=DefaultFunction
      targetScale=0.1361
      boneTransforms [{targetRotation=<rot -0.2203 114.6092 -0.6612>, targetTranslation=<-26.4651 0.8345 -0.259>}]
      customCurve {defaultValue=3.4028234663852886e+38, postInfinityExtrap=RCCE_Constant, preInfinityExtrap=RCCE_Constant, keys []}
      targetRotation=<rot 0 0 0>
     }
     {
      bApplyCustomCurve=false
      bIsHidden=false
      distanceMethod=DefaultMethod
      drivenName=Pose_2
      functionType=DefaultFunction
      targetScale=0.1387
      boneTransforms [{targetRotation=<rot -0.0782 107.8265 -0.5542>, targetTranslation=<-26.4651 0.8345 -0.259>}]
      customCurve {defaultValue=3.4028234663852886e+38, postInfinityExtrap=RCCE_Constant, preInfinityExtrap=RCCE_Constant, keys []}
      targetRotation=<rot 0 0 0>
     }
     {
      bApplyCustomCurve=false
      bIsHidden=false
      distanceMethod=DefaultMethod
      drivenName=Pose_3
      functionType=DefaultFunction
      targetScale=0.0329
      boneTransforms [{targetRotation=<rot -0.1853 110.7678 -0.5612>, targetTranslation=<-26.4651 0.8345 -0.259>}]
      customCurve {defaultValue=3.4028234663852886e+38, postInfinityExtrap=RCCE_Constant, preInfinityExtrap=RCCE_Constant, keys []}
      targetRotation=<rot 0 0 0>
     }
     {
      bApplyCustomCurve=false
      bIsHidden=false
      distanceMethod=DefaultMethod
      drivenName=Pose_4
      functionType=DefaultFunction
      targetScale=1
      boneTransforms [{targetRotation=<rot 0.6608 47.7825 0.7637>, targetTranslation=<-26.4651 0.8345 -0.259>}]
      customCurve {defaultValue=3.4028234663852886e+38, postInfinityExtrap=RCCE_Constant, preInfinityExtrap=RCCE_Constant, keys []}
      targetRotation=<rot 0 0 0>
     }
     {
      bApplyCustomCurve=false
      bIsHidden=false
      distanceMethod=DefaultMethod
      drivenName=Pose_5
      functionType=DefaultFunction
      targetScale=0.3105
      boneTransforms [{targetRotation=<rot -0.282 121.1986 -0.8364>, targetTranslation=<-26.4651 0.8345 -0.259>}]
      customCurve {defaultValue=3.4028234663852886e+38, postInfinityExtrap=RCCE_Constant, preInfinityExtrap=RCCE_Constant, keys []}
      targetRotation=<rot 0 0 0>
     }
     {
      bApplyCustomCurve=false
      bIsHidden=false
      distanceMethod=DefaultMethod
      drivenName=Pose_6
      functionType=DefaultFunction
      targetScale=0.0314
      boneTransforms [{targetRotation=<rot 0.0995 111.7848 -0.1452>, targetTranslation=<-26.4651 0.8345 -0.259>}]
      customCurve {defaultValue=3.4028234663852886e+38, postInfinityExtrap=RCCE_Constant, preInfinityExtrap=RCCE_Constant, keys []}
      targetRotation=<rot 0 0 0>
     }
     {
      bApplyCustomCurve=false
      bIsHidden=false
      distanceMethod=DefaultMethod
      drivenName=Pose_7
      functionType=DefaultFunction
      targetScale=0.0314
      boneTransforms [{targetRotation=<rot 0.0561 111.132 -0.0175>, targetTranslation=<-26.4651 0.8345 -0.259>}]
      customCurve {defaultValue=3.4028234663852886e+38, postInfinityExtrap=RCCE_Constant, preInfinityExtrap=RCCE_Constant, keys []}
      targetRotation=<rot 0 0 0>
     }
    ]
    rBFParams {distanceMethod=Quaternion, radius=55}
    sourceBones [{boneName=lowerarm_r}]
   }
   n37 = Anim Dynamics - Bone: elbow_outer_l(ComponentPose=Local To Component(LocalPose=@n36), Alpha=0.55, ExternalForce="0,0,0") {
    bGravityOverrideInSimSpace=true
    gravityScale=0.3
    linearDampingOverride=0.9999
    linearSpringConstant=100
    alphaBoolBlend {blendOption=Linear}
    boundBone {boneName=elbow_outer_l}
    physicsBodyDefinitions [{
      sphereCollisionRadius=0
      boundBone {boneName=elbow_outer_l}
      boxExtents=<20 20 20>
      constraintSetup {angularLimitsMax=<0 0 0>, angularLimitsMin=<0 0 0>, angularTarget=<0 0 0>, linearAxesMax=<0.5 2 1>, linearAxesMin=<-0.5 -2 -1>}
      localJointOffset=<-7 0 0>
     }]
   }
   n38 = Anim Dynamics - Bone: elbow_outer_r(ComponentPose=@n37, Alpha=0.55, ExternalForce="0,0,0") {
    bGravityOverrideInSimSpace=true
    gravityScale=0.3
    linearDampingOverride=0.9999
    linearSpringConstant=100
    alphaBoolBlend {blendOption=Linear}
    boundBone {boneName=elbow_outer_r}
    physicsBodyDefinitions [{
      sphereCollisionRadius=0
      boundBone {boneName=elbow_outer_r}
      boxExtents=<20 20 20>
      constraintSetup {angularLimitsMax=<0 0 0>, angularLimitsMin=<0 0 0>, angularTarget=<0 0 0>, linearAxesMax=<0.5 2 1>, linearAxesMin=<-0.5 -2 -1>}
      localJointOffset=<7 0 0>
     }]
   }
   n39 = "Transform (Modify) Bone - Bone: driver_apron_front_01_l"!(ComponentPose=@n38, Alpha:float=1, Translation="0,0,4.2", Rotation="0,0,0", Scale="1,1,1") -> (Pose) {
    rotationMode=BMM_Additive
    rotationSpace=BCS_BoneSpace
    translationMode=BMM_Additive
    translationSpace=BCS_BoneSpace
    alphaBoolBlend {blendOption=Linear}
    boneToModify {boneName=driver_apron_front_01_l}
    translation {z=4.2}
   }
   n40 = "Transform (Modify) Bone - Bone: driver_apron_front_01_r"!(ComponentPose=@n39, Alpha:float=1, Translation="0,0,0", Rotation="0.000,-10,0", Scale="1,1,1") -> (Pose) {
    rotationMode=BMM_Additive
    rotationSpace=BCS_BoneSpace
    translationMode=BMM_Additive
    translationSpace=BCS_BoneSpace
    alphaBoolBlend {blendOption=Linear}
    boneToModify {boneName=driver_apron_front_01_r}
    rotation {yaw=-10}
   }
   Save cached pose 'PreSim_Cache'(Pose=Component To Local(ComponentPose=@n40)) {cacheName=PreSim_Cache}
  ]}
 EventGraph {nodes [
   Event BlueprintInitializeAnimation()
   > n1 = Cast To ClothingSimulationInteractorNv!(
    execute
    Object=GetClothingSimulationInteractor!(self:Skeletal Mesh Component Object Reference=GetPartSkeletalMeshComponent(), ClothingAssetName:Name=None) -> (ReturnValue:Clothing Simulation Interactor Object Reference)
   ) -> (then, CastFailed, AsClothing Simulation Interactor Nv:Clothing Simulation Interactor Nv Object Reference) {is=title K2Node_DynamicCast}
   > Set clothing_sim_interactor(clothing_sim_interactor:Clothing Simulation Interactor Nv Object Reference=@n1) -> (Output_Get:Clothing Simulation Interactor Nv Object Reference)
   > Set MeshToCopy(MeshToCopy:Skeletal Mesh Component Object Reference=GetOwnerSkeletalMeshComponent()) -> (Output_Get:Skeletal Mesh Component Object Reference)
   Event OnBeginJumping()
   > Set Jumping(Jumping:bool=true) -> (Output_Get:bool)
   Event OnBeginSkydiving()
   > Set is_skydiving(is_skydiving:bool=true) -> (Output_Get:bool)
   Event OnEndSkydiving()
   > Set is_skydiving(is_skydiving=false)
   Event OnBeginCrouching()
   > Set is_crouching(is_crouching:bool=true) -> (Output_Get:bool)
   > Set crouching_alpha(crouching_alpha:double=1) -> (Output_Get:double)
   > Set neg_crouching_alpha(neg_crouching_alpha:double=0.2) -> (Output_Get:double)
   > Set neg_crouching_alpha_A(neg_crouching_alpha_A:double=0.1) -> (Output_Get:double)
   Event OnEndCrouching()
   > Set is_crouching(is_crouching=false)
   > Set crouching_alpha(crouching_alpha=0)
   > Set neg_crouching_alpha(neg_crouching_alpha=1)
   > Set neg_crouching_alpha_A(neg_crouching_alpha_A=0.3)
   Event BlueprintUpdateAnimation()
   > n2 = Get!(execute:Exec) -> (then:Exec, else:Exec, MeshToCopy:Skeletal Mesh Component Object Reference)
   n3 = Sequence!(execute=[@n2.else, @n2.then]) -> (then_0, then_1, then_2:Exec, then_3:Exec, then_4:Exec, then_5:Exec, then_6:Exec)
   > n4 = Branch(Condition=Get is_skydiving() -> (is_skydiving:bool))
   n5 = Get clothing_sim_interactor() -> (clothing_sim_interactor:Clothing Simulation Interactor Nv Object Reference)
   > n6 = EnableGravityOverride!(
    execute
    self:Clothing Simulation Interactor Object Reference=@n5
    InVector:"Vector (by ref)"=Break Fort Anim Input Skydiving External Force(FortAnimInput_SkydivingExternalForce=Get SkydivingExternalForceInput()) -> (
     bUseSkydivingVectorForce=~
     SkydivingVectorForce=~
     HeadToPelvisDirection=~
     FloatingMultiplier=~
     FloatingAdditive=~
     DivingMultiplier=~
     DivingAdditive=~
     ParachutingMultiplier=~
     ParachutingAdditive=~
    ).SkydivingVectorForce
   ) -> (then)
   ForceClothNextUpdateTeleportAndReset(
    execute=[@n6.then, @n7.Completed]
    self:Skeletal Mesh Component Object Reference=Get MeshToCopy() -> (MeshToCopy:Skeletal Mesh Component Object Reference)
   )
   DisableGravityOverride(execute=@n4.else, self:Clothing Simulation Interactor Object Reference=@n5)
   > n7 = Do Once()
   n8 = Sequence!(execute=@n3.then_1) -> (then_0, then_1, then_2)
   n9 = Get bIsInFrontEnd()
   > n10 = Branch(Condition=@n9)
   n11 = GetOwnerSkeletalMeshComponent()
   n12 = BreakRotator(InRot=BreakTransform(InTransform=GetDeltaTransformFromRefPose(self=@n11, BoneName=thigh_l, BaseName=pelvis)).Rotation)
   > Set FrontCloth01_L(FrontCloth01_L:Rotator=MakeRotator(Yaw=MapRangeClamped(Value=@n12.Yaw, InRangeA=0, InRangeB=-50, OutRangeB=-50))) -> (Output_Get:Rotator)
   Set FrontCloth01_L(execute=@n10.else, FrontCloth01_L=MakeRotator(Yaw=MapRangeClamped(Value=@n12.Yaw, InRangeA=-25, InRangeB=-90, OutRangeA=3, OutRangeB=-30)))
   n13 = Branch(execute=@n8.then_1, Condition=@n9)
   > Set FlontCloth02_L(FlontCloth02_L:Rotator=MakeRotator(Yaw=MapRangeClamped(Value=@n12.Yaw, InRangeA=0, InRangeB=-50, OutRangeB=-20))) -> (Output_Get:Rotator)
   Set FlontCloth02_L(execute=@n13.else, FlontCloth02_L=MakeRotator(Yaw=MapRangeClamped(Value=@n12.Yaw, InRangeA=-25, InRangeB=-40, OutRangeA=-10, OutRangeB=-30)))
   n14 = Branch(execute=@n3.then_3, Condition=Get bIsJumping())
   > Set jumpAlpha(jumpAlpha:double=0.2) -> (Output_Get:double)
   Set jumpAlpha(execute=@n14.else, jumpAlpha=1)
   n15 = Branch(execute=@n3.then_4, Condition=Get bIsMoving2D())
   > Set Moving2D(Moving2D:bool=true) -> (Output_Get:bool)
   Set Moving2D(execute=@n15.else, Moving2D=false)
   Set CrouchingDynamics(execute=@n3.then_5, CrouchingDynamics:bool=OR Boolean(A=Get is_crouching() -> (is_crouching:bool), B=Get bIsDBNO())) -> (Output_Get:bool)
   n16 = GetOwnerSkeletalMeshComponent()
   n17 = BreakRotator(InRot=BreakTransform(InTransform=GetDeltaTransformFromRefPose(self=@n16, BoneName=pelvis, BaseName=thigh_r)).Rotation)
   Set thigh_inv_float_r(execute=@n3.then_2, thigh_inv_float_r:double=MapRangeClamped(Value=@n17.Pitch, InRangeA=0, InRangeB=-30)) -> (Output_Get:double)
   > Set thigh_float_r(thigh_float_r:double=MapRangeClamped(Value=@n17.Yaw, InRangeA=10, InRangeB=120)) -> (Output_Get:double)
   > n18 = Sequence!(execute) -> (then_0, then_1)
   n19 = Get bIsInFrontEnd()
   > n20 = Branch(Condition=@n19)
   n21 = BreakRotator(InRot=BreakTransform(InTransform=GetDeltaTransformFromRefPose(self=@n16, BoneName=thigh_r, BaseName=pelvis)).Rotation)
   > Set FrontCloth01_R(FrontCloth01_R:Rotator=MakeRotator(Yaw=MapRangeClamped(Value=@n21.Yaw, InRangeA=0, InRangeB=-50, OutRangeB=-50))) -> (Output_Get:Rotator)
   Set FrontCloth01_R(execute=@n20.else, FrontCloth01_R=MakeRotator(Yaw=MapRangeClamped(Value=@n21.Yaw, InRangeA=-3, InRangeB=-70, OutRangeA=3, OutRangeB=-30)))
   n22 = Branch(execute=@n18.then_1, Condition=@n19)
   > Set FlontCloth02_R(FlontCloth02_R:Rotator=MakeRotator(Yaw=MapRangeClamped(Value=@n21.Yaw, InRangeA=0, InRangeB=-50, OutRangeB=-20))) -> (Output_Get:Rotator)
   Set FlontCloth02_R(execute=@n22.else, FlontCloth02_R=MakeRotator(Yaw=MapRangeClamped(Value=@n21.Yaw, InRangeA=3, InRangeB=-40, OutRangeB=-15)))
   Event OnEndFalling()
   > Set Jumping(Jumping=false)
   n23 = Branch(Condition=Get is_skydiving())
   n24 = Branch(execute=[@n23.else, @n23.then], Condition=Get bIsSkydiveFloating())
   n25 = Get Direction Between Sockets!(
    Component:Skeletal Mesh Component Object Reference=GetOwnerSkeletalMeshComponent()
    SocketOrBoneNameFrom:Name=head
    SocketOrBoneNameTo:Name=pelvis
   ) -> (ReturnValue:Vector) {is=title K2Node_CallFunction}
   > Set skydiving vector force(skydiving vector force:Vector=vector + vector(
     A:Vector=vector * vector(A:Vector=@n25, B:Vector="75,75,75") -> (ReturnValue:Vector) {is=title K2Node_CommutativeAssociativeBinaryOperator}
     B:Vector="0, 0, 0"
    ) -> (ReturnValue:Vector) {is=title K2Node_CommutativeAssociativeBinaryOperator}) -> (Output_Get:Vector)
   n26 = Branch(execute=@n24.else, Condition=Get bIsSkydiveDiveMode())
   > Set skydiving vector force(skydiving vector force=vector + vector(A=vector * vector(A=@n25, B="180,180,200"), B="0, 0,1000"))
   Set skydiving vector force(execute=@n26.else, skydiving vector force=vector * vector(A=@n25, B="75,75,75"))
   n27 = GetDeltaTransformFromRefPose(self=GetOwnerSkeletalMeshComponent(), BoneName=pelvis, BaseName=thigh_l)
   MapRangeClamped(Value=BreakRotator(InRot=BreakTransform(InTransform=@n27).Rotation).Yaw, InRangeA=10, InRangeB=120)
   Event OnCharacterCustomizationLoaded()
   Reroute Node(InputPin:Vector=vector * vector(A="0, 0, 0", B="-5,-5,5")) -> (OutputPin:Vector)
   Set Body_Mesh(Body_Mesh=GetPartSkeletalMeshComponent())
   Set thigh_inv_float_l(thigh_inv_float_l:double=MapRangeClamped(Value=0, InRangeA=-25, InRangeB=-100)) -> (Output_Get:double)
  ]}
]
variables [
 Body_Mesh {type=SkeletalMeshComponent}
 clothing_sim_interactor {type=ClothingSimulationInteractorNv, default=null}
 crouching_alpha {type=number}
 CrouchingDynamics {type=boolean}
 FlontCloth02_L {type=Rotator}
 FlontCloth02_R {type=Rotator}
 FrontCloth01_L {type=Rotator}
 FrontCloth01_R {type=Rotator}
 is_crouching {type=boolean}
 is_skydiving {type=boolean}
 jumpAlpha {type=number}
 Jumping {type=boolean}
 MeshToCopy {type=SkeletalMeshComponent}
 Moving2D {type=boolean}
 neg_crouching_alpha {type=number, default=1}
 neg_crouching_alpha_A {type=number, default=1}
 skydiving vector force {type=Vector}
 thigh_float_l {type=number}
 thigh_float_r {type=number}
 thigh_inv_float_l {type=number}
 thigh_inv_float_r {type=number}
]
