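# Physics asset for the bow on the Sushi Ninja medium female cosmetic body, defining collision and joints for its dyn_bow_top bones.
# Small capsules cover two bones per side and spine_02, with a small sphere on the pelvis.
# The dyn_bow_top_l/r bodies constrain to spine_02, while the 01 bodies constrain directly to the pelvis rather than chaining.
# The spine_02-to-pelvis joint leaves swing and twist free.
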PhysicsAsset /BRCosmetics/Characters/Player/Female/Medium/Bodies/F_MED_Sushi_Ninja/Meshes/F_MED_Sushi_Ninja_Bow_Physics
bodies {
 dyn_bow_top_01_l [{shapeType=Capsule, length=3, radius=2.5, center=<0.1199 -0.0001 0.4473>, rotation=<rot -15 0 0>}]
 dyn_bow_top_01_r [{shapeType=Capsule, length=5, radius=2.5, center=<-0.2137 -0.0001 0.7975>, rotation=<rot 15 0 0>}]
 dyn_bow_top_l [{shapeType=Capsule, length=3.75, radius=2.25, center=<0.2774 0 1.0353>, rotation=<rot -15 0 0>}]
 dyn_bow_top_r [{shapeType=Capsule, length=2.75, radius=2.75, center=<-0.2241 0 0.8365>, rotation=<rot 15 0 0>}]
 pelvis [{shapeType=Sphere, radius=1}]
 spine_02 [{shapeType=Capsule, length=4.8758, radius=2.8841, center=<2.2074 0.1071 0.0584>, rotation=<rot 0 5.2715 -173.6998>}]
}
constraints [
 {bone1Name=dyn_bow_top_l, bone2Name=spine_02, swing1LimitDegrees=45, swing2LimitDegrees=45, twistLimitDegrees=45}
 {bone1Name=dyn_bow_top_r, bone2Name=spine_02, swing1LimitDegrees=45, swing2LimitDegrees=45, twistLimitDegrees=45}
 {
  bone1Name=spine_02
  bone2Name=pelvis
  swing1LimitDegrees=45
  swing1Motion=Free
  swing2LimitDegrees=45
  swing2Motion=Free
  twistLimitDegrees=45
  twistMotion=Free
 }
 {bone1Name=dyn_bow_top_01_l, bone2Name=pelvis, swing1LimitDegrees=45, swing2LimitDegrees=45, twistLimitDegrees=45}
 {bone1Name=dyn_bow_top_01_r, bone2Name=pelvis, swing1LimitDegrees=45, swing2LimitDegrees=45, twistLimitDegrees=45}
]
