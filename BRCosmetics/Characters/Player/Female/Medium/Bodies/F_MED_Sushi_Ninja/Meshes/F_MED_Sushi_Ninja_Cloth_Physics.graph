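# Physics asset for the Sushi Ninja female medium body cosmetic, named for cloth physics, providing simplified body collision shapes.
# Capsules cover the pelvis, both thighs and the left forearm, with several overlapping capsules shaping the pelvis.
# Both thighs constrain to the pelvis with angular limits, while the left forearm's constraint to the pelvis leaves swing and twist free.
# No right-arm or upper-torso bodies are present.
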PhysicsAsset /BRCosmetics/Characters/Player/Female/Medium/Bodies/F_MED_Sushi_Ninja/Meshes/F_MED_Sushi_Ninja_Cloth_Physics
bodies {
 lowerarm_l [{shapeType=Capsule, length=4.866, radius=4.1658, center=<3.4652 -0.0227 0.2409>, rotation=<rot -22.7333 -97.5978 89.0914>}]
 pelvis [
  {shapeType=Capsule, length=13.8798, radius=10.0665, center=<-4.725 -2.8839 2.8251>, rotation=<rot 90 0 -10.5557>}
  {shapeType=Capsule, length=13.25, radius=10.75, center=<-9.5884 1.0145 0.0003>, rotation=<rot 0.2151 177.9471 177.9471>}
  {shapeType=Capsule, length=13.8798, radius=10.0665, center=<-4.3049 -2.541 -3.954>, rotation=<rot 90 0 -10.5557>}
  {shapeType=Capsule, length=14, radius=10, center=<-1.1206 -0.6199 0.0006>, rotation=<rot 0.2151 177.9471 177.9471>}
 ]
 thigh_l [{shapeType=Capsule, length=23.1417, radius=9.1044, center=<-18.4814 0.1296 0.1501>, rotation=<rot 0.6014 -87.9846 -91.106>}]
 thigh_r [{shapeType=Capsule, length=22.9227, radius=9.0851, center=<18.2527 0.0222 -0.2736>, rotation=<rot 0.4904 -89.3386 -91.1652>}]
}
constraints [
 {bone1Name=thigh_l, bone2Name=pelvis, swing1LimitDegrees=45, swing2LimitDegrees=45, twistLimitDegrees=45}
 {bone1Name=thigh_r, bone2Name=pelvis, swing1LimitDegrees=45, swing2LimitDegrees=45, twistLimitDegrees=45}
 {
  bone1Name=lowerarm_l
  bone2Name=pelvis
  swing1LimitDegrees=45
  swing1Motion=Free
  swing2LimitDegrees=45
  swing2Motion=Free
  twistLimitDegrees=45
  twistMotion=Free
 }
]
