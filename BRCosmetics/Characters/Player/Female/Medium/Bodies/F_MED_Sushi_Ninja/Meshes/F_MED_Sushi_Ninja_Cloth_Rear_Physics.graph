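# Physics asset for the rear cloth piece on the female medium Sushi Ninja body, giving it jointed secondary motion.
# A small base capsule is locked to the pelvis and anchors a two-box middle chain with limited swing and zero twist limit.
# The lower middle box branches into two angled side boxes that allow limited swing and twist.
# Pelvis and thigh capsules provide leg and hip collision shapes, with the thighs rotationally free relative to the pelvis.
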PhysicsAsset /BRCosmetics/Characters/Player/Female/Medium/Bodies/F_MED_Sushi_Ninja/Meshes/F_MED_Sushi_Ninja_Cloth_Rear_Physics
bodies {
 base_cloth_rear_mid [{shapeType=Capsule, length=0.505, radius=0.505, shapeName=base_cloth_rear_mid_capsule, rotation=<rot 0 -90 0>}]
 dyn_cloth_rear_left [{
   shapeType=Box
   extentX=2.5
   extentY=10.0897
   extentZ=11.7545
   shapeName=dyn_cloth_rear_left_box
   center=<4 0 -2>
   rotation=<rot -3.0001 -89.9999 42>
  }]
 dyn_cloth_rear_mid_01 [{shapeType=Box, extentX=14, extentY=1.5, extentZ=10, center=<0 -0.7165 -4.6046>}]
 dyn_cloth_rear_mid_02 [{shapeType=Box, extentX=18, extentY=1.5, extentZ=9.5, center=<0 -0.6467 -1.115>, rotation=<rot 0 0 -3.4128>}]
 dyn_croth_rear_reft [{
   shapeType=Box
   extentX=0.9361
   extentY=13.3988
   extentZ=10.4599
   shapeName=dyn_croth_rear_reft_box
   center=<-4 0 2>
   rotation=<rot 2.9998 -89.9998 -48>
  }]
 pelvis [{shapeType=Capsule, length=14.25, radius=11, center=<-4.3002 -1.32 0>}]
 thigh_l [{shapeType=Capsule, length=12.2422, radius=10.7, center=<-14.6322 -2.0637 1.5098>, rotation=<rot 0.4884 -93.8431 -91.8795>}]
 thigh_r [{shapeType=Capsule, length=14.1298, radius=11.95, center=<14.7323 0.2555 -2.4442>, rotation=<rot -0.1621 -98.5112 -93.5971>}]
}
constraints [
 {
  bone1Name=thigh_l
  bone2Name=pelvis
  swing1LimitDegrees=45
  swing1Motion=Free
  swing2LimitDegrees=45
  swing2Motion=Free
  twistLimitDegrees=45
  twistMotion=Free
 }
 {
  bone1Name=thigh_r
  bone2Name=pelvis
  swing1LimitDegrees=45
  swing1Motion=Free
  swing2LimitDegrees=45
  swing2Motion=Free
  twistLimitDegrees=45
  twistMotion=Free
 }
 {bone1Name=dyn_cloth_rear_mid_01, bone2Name=base_cloth_rear_mid, swing1LimitDegrees=25, swing2LimitDegrees=20, twistLimitDegrees=0}
 {bone1Name=dyn_cloth_rear_mid_02, bone2Name=dyn_cloth_rear_mid_01, swing1LimitDegrees=15, swing2LimitDegrees=25, twistLimitDegrees=0}
 {
  bone1Name=base_cloth_rear_mid
  bone2Name=pelvis
  swing1LimitDegrees=45
  swing1Motion=Locked
  swing2LimitDegrees=45
  swing2Motion=Locked
  twistLimitDegrees=45
  twistMotion=Locked
 }
 {bone1Name=dyn_cloth_rear_left, bone2Name=dyn_cloth_rear_mid_02, swing1LimitDegrees=20, swing2LimitDegrees=30, twistLimitDegrees=30}
 {bone1Name=dyn_croth_rear_reft, bone2Name=dyn_cloth_rear_mid_02, swing1LimitDegrees=20, swing2LimitDegrees=30, twistLimitDegrees=30}
]
