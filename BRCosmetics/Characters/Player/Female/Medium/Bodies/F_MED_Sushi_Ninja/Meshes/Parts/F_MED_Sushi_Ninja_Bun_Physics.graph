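# Physics asset for the hair bun part of the female medium Sushi Ninja cosmetic body, covering a hair bun and a neck knot.
# A capsule on the bun bone is constrained to the head, which carries two overlapping capsules, with equal narrow swing and twist limits.
# Left and right knot capsules each hang off a small central knot sphere with asymmetric swing limits.
# No constraint links the knot group to the head.
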PhysicsAsset /BRCosmetics/Characters/Player/Female/Medium/Bodies/F_MED_Sushi_Ninja/Meshes/Parts/F_MED_Sushi_Ninja_Bun_Physics
bodies {
 dyn_hair_bun [{shapeType=Capsule, length=6.5, radius=5, center=<1 -3.3035 2.6224>, rotation=<rot 0 0 -40.0002>}]
 dyn_neck_knot [{shapeType=Sphere, radius=0.505}]
 dyn_neck_knot_l [{shapeType=Capsule, length=1.75, radius=0.75, center=<0.124 0 0>, rotation=<rot -30.0001 0 0>}]
 dyn_neck_knot_r [{shapeType=Capsule, length=1, radius=1, center=<-0.14 0 0>, rotation=<rot 30.0001 0 0>}]
 head [
  {shapeType=Capsule, length=11.7937, radius=6.2123, rotation=<rot -90 -165.9637 -164.0347>}
  {shapeType=Capsule, length=8.7937, radius=5.2123, center=<-2.7498 0.5541 0.0001>, rotation=<rot -90 -14.0363 86.7341>}
 ]
}
constraints [
 {bone1Name=dyn_hair_bun, bone2Name=head, swing1LimitDegrees=15, swing2LimitDegrees=15, twistLimitDegrees=15}
 {bone1Name=dyn_neck_knot_l, bone2Name=dyn_neck_knot, swing1LimitDegrees=20, swing2LimitDegrees=45, twistLimitDegrees=15}
 {bone1Name=dyn_neck_knot_r, bone2Name=dyn_neck_knot, swing1LimitDegrees=20, swing2LimitDegrees=45, twistLimitDegrees=15}
]
