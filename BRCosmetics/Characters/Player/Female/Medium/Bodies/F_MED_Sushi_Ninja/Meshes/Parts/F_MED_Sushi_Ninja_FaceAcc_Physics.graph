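# Physics asset for the face accessory part of the female medium Sushi Ninja character, covering hair and neck-knot bones.
# Four box bodies form a hair chain constrained link to link from the head, which has two capsules.
# Left and right capsule bodies each constrain to a central neck-knot sphere, which has no constraint to the head.
# All joints set limited swing and twist angles, with the wider swing on the second axis.
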PhysicsAsset /BRCosmetics/Characters/Player/Female/Medium/Bodies/F_MED_Sushi_Ninja/Meshes/Parts/F_MED_Sushi_Ninja_FaceAcc_Physics
bodies {
 dyn_neck_knot [{shapeType=Sphere, radius=0.505}]
 dyn_neck_knot_l [{shapeType=Capsule, length=1.75, radius=0.75, center=<0.124 0 0>, rotation=<rot -30.0001 0 0>}]
 dyn_neck_knot_r [{shapeType=Capsule, length=1, radius=1, center=<-0.14 0 0>, rotation=<rot 30.0001 0 0>}]
 dyn_sn_hair_01 [{shapeType=Box, extentX=1.01, extentY=5.51, extentZ=7.51, center=<1.0206 0 -0.0047>, rotation=<rot 0 -89.9999 0.2636>}]
 dyn_sn_hair_02 [{shapeType=Box, extentX=0.9022, extentY=4.044, extentZ=7.0466, center=<1.7672 0.1035 -0.4363>, rotation=<rot -13.5451 -90.051 4.9267>}]
 dyn_sn_hair_03 [{shapeType=Box, extentX=1.3095, extentY=3.3827, extentZ=5.3943, center=<1.6262 -0.1012 0.2926>, rotation=<rot -10.0716 -89.3876 -1.0364>}]
 dyn_sn_hair_04 [{shapeType=Box, extentX=0.8946, extentY=2.7698, extentZ=4.9422, center=<1.9222 0.1296 0.1474>, rotation=<rot 0 -83.7709 -87.1731>}]
 head [
  {shapeType=Capsule, length=11.7937, radius=6.2123, rotation=<rot -90 -165.9637 -164.0347>}
  {shapeType=Capsule, length=8.7937, radius=5.2123, center=<-2.7498 0.5541 0.0001>, rotation=<rot -90 -14.0363 86.7341>}
 ]
}
constraints [
 {bone1Name=dyn_sn_hair_01, bone2Name=head, swing1LimitDegrees=15, swing2LimitDegrees=45, twistLimitDegrees=14.1429}
 {bone1Name=dyn_sn_hair_02, bone2Name=dyn_sn_hair_01, swing1LimitDegrees=15, swing2LimitDegrees=45, twistLimitDegrees=8}
 {bone1Name=dyn_sn_hair_03, bone2Name=dyn_sn_hair_02, swing1LimitDegrees=15, swing2LimitDegrees=45, twistLimitDegrees=8}
 {bone1Name=dyn_sn_hair_04, bone2Name=dyn_sn_hair_03, swing1LimitDegrees=16, swing2LimitDegrees=45, twistLimitDegrees=8}
 {bone1Name=dyn_neck_knot_l, bone2Name=dyn_neck_knot, swing1LimitDegrees=20, swing2LimitDegrees=45, twistLimitDegrees=15}
 {bone1Name=dyn_neck_knot_r, bone2Name=dyn_neck_knot, swing1LimitDegrees=20, swing2LimitDegrees=45, twistLimitDegrees=15}
]
