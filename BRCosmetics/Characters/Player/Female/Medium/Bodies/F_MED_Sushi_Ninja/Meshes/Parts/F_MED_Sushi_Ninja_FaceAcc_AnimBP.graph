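# Post-process style Anim Blueprint for the female Sushi Ninja face accessory, adding secondary motion to the neck knot, hair bun and chopsticks.
# It copies the attached parent's mesh pose and curves, then drives the knot and chopsticks with weighted multi-bone constraints, nudging each chopstick apart additively.
# Two chained RigidBody nodes then simulate the accessory and bun physics assets in spine and neck space, the bun at half alpha.
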AnimBlueprint /BRCosmetics/Characters/Player/Female/Medium/Bodies/F_MED_Sushi_Ninja/Meshes/Parts/F_MED_Sushi_Ninja_FaceAcc_AnimBP
target_skeleton=&/F_MED_Sushi_Ninja_FaceAcc_Skeleton
graphs [
 AnimGraph {nodes [
   n1 = "Constraint - dyn_neck_knot (PPPP)"!(
    ComponentPose=Local To Component(LocalPose=Copy Pose From Mesh() {bCopyCurves=true, bUseAttachedParent=true})
    Alpha=1
    ConstraintWeights_0=0.339
    ConstraintWeights_1:float=0.0512
    ConstraintWeights_2:float=0.5913
    ConstraintWeights_3:float=0.0185
   ) -> (Pose) {
    alphaBoolBlend {blendOption=Linear}
    boneToModify {boneName=dyn_neck_knot}
    constraintSetup [{targetBone {boneName=neck_02}}, {targetBone {boneName=neck_01}}, {targetBone {boneName=head}}, {targetBone {boneName=C_jaw}}]
    constraintWeights [1,1,1,1]
   }
   n2 = "Constraint - dyn_chopstick_a (PP)"!(ComponentPose=@n1, Alpha=1, ConstraintWeights_0=0.65, ConstraintWeights_1:float=0.35) -> (Pose) {
    alphaBoolBlend {blendOption=Linear}
    boneToModify {boneName=dyn_chopstick_a}
    constraintSetup [{targetBone {boneName=head}}, {targetBone {boneName=dyn_hair_bun}}]
    constraintWeights [1,1]
   }
   n3 = "Transform (Modify) Bone - Bone: dyn_chopstick_a"(ComponentPose=@n2, Alpha=1, Translation="0,-0.2,0.2", Rotation="0,0,0", Scale="1,1,1") {
    translationMode=BMM_Additive
    translationSpace=BCS_BoneSpace
    alphaBoolBlend {blendOption=Linear}
    boneToModify {boneName=dyn_chopstick_a}
    translation {y=-0.2, z=0.2}
   }
   n4 = "Constraint - dyn_chopstick_b (PP)"!(ComponentPose=@n3, Alpha=1, ConstraintWeights_0=0.65, ConstraintWeights_1:float=0.35) -> (Pose) {
    alphaBoolBlend {blendOption=Linear}
    boneToModify {boneName=dyn_chopstick_b}
    constraintSetup [{targetBone {boneName=head}}, {targetBone {boneName=dyn_hair_bun}}]
    constraintWeights [1,1]
   }
   n5 = "Transform (Modify) Bone - Bone: dyn_chopstick_b"(ComponentPose=@n4, Alpha=1, Translation="0,0.2,-0.2", Rotation="0,0,0", Scale="1,1,1") {
    translationMode=BMM_Additive
    translationSpace=BCS_BoneSpace
    alphaBoolBlend {blendOption=Linear}
    boneToModify {boneName=dyn_chopstick_b}
    translation {y=0.2, z=-0.2}
   }
   n6 = RigidBody(ComponentPose=@n5) {
    bClampLinearTranslationLimitToRefPose=true
    bOverrideWorldGravity=true
    overridePhysicsAsset=&/F_MED_Sushi_Ninja_FaceAcc_Physics
    simulationSpace=BaseBoneSpace
    alphaBoolBlend {blendOption=Linear}
    baseBoneRef {boneName=spine_05}
    componentAppliedLinearAccClamp=<1100 1100 1050>
    componentLinearAccScale=<0.7 0.7 0.3>
    componentLinearVelScale=<0.7 0.7 0.3>
    overrideWorldGravity {z=-980}
   }
   n7 = RigidBody(ComponentPose=@n6, Alpha=0.5) {
    bClampLinearTranslationLimitToRefPose=true
    bOverrideWorldGravity=true
    overridePhysicsAsset=&/F_MED_Sushi_Ninja_Bun_Physics
    simulationSpace=BaseBoneSpace
    alphaBoolBlend {blendOption=Linear}
    baseBoneRef {boneName=neck_01}
    componentAppliedLinearAccClamp=<1100 1100 1050>
    componentLinearAccScale=<0.7 0.7 0.3>
    componentLinearVelScale=<0.7 0.7 0.3>
    overrideWorldGravity {z=-980}
   }
   Output Pose(Result=Component To Local(ComponentPose=@n7)) {layerGroup=None}
  ]}
 EventGraph {nodes [
   TryGetPawnOwner()
   Event BlueprintUpdateAnimation()
  ]}
]
variables []
